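# Unlit translucent particle material for lightning arcs on the PlatypusBranch pickaxe, with Niagara mesh particle support.
# Layered panning noise, randomized per particle, distorts UVs feeding two GeneratedBand masks, forming a core and soft blur with base-to-tip width scalars.
# Opacity adds growth, HeightLerp dissolve and quality-switched depth fade, while emissive boosts the tinted core.
# World position offset combines SplineThicken with sine oscillation, modulated by dynamic parameters.
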Material /BRCosmetics/Weapons/FORT_Melee/Pickaxe_PlatypusBranch/FX/M_Pickaxe_PlatypusBranch_Lightning
graph {
 properties {
  bUsedWithNiagaraMeshParticles=true
  blendMode=BLEND_Translucent
  pixelDepthOffsetMode=PDOM_Legacy
  shadingModel=MSM_Unlit
  twoSided=true
  usedShadingModels=Unlit
 }
 nodes [
  n1 = TextureCoordinate() {vTiling=2}
  n2 = ScalarParameter() {defaultValue=1, group=WPO Oscillation, parameterName=Sin Frequency, sortPriority=1}
  n3 = DynamicParameter() {parameterIndex=1, defaultValue=<rgba 0 1 1 1>, paramNames [Dissolve, Noise Frequency, Sin Frequency, Noise Speed]}
  n4 = Add(A=AppendVector(A=Constant() {r=1}, B=Multiply(A=@n2, B=@n3.Sin Frequency)), B=ParticleRandom())
  NamedRerouteDeclaration(None=ComponentMask(None=Multiply(A=@n1, B=@n4)) {g=true}) {name=UV.v, nodeColor=<rgba 1 0 0.7266 1>}
  n5 = ParticleColor()
  NamedRerouteDeclaration(None=@n5.A) {name=Particle Alpha, nodeColor=<rgba 0.8984 1 0 1>}
  n6 = ScalarParameter() {group=Base to Tip, parameterName=Tip Width Scalar, sortPriority=3}
  n7 = ComponentMask(None=TextureCoordinate()) {g=true}
  n8 = Power(Base=@n7, Exp=ScalarParameter() {defaultValue=1, group=Base To Tip, parameterName=Tip Exponent, sortPriority=2}) {constExponent=1}
  n9 = ScalarParameter() {group=Base to Tip, parameterName=Base Width Scalar, sortPriority=1}
  n10 = Power(Base=OneMinus(None=@n7), Exp=ScalarParameter() {defaultValue=1, group=Base to Tip, parameterName=Base Exponent, sortPriority=0})
  NamedRerouteDeclaration(None=Multiply(A=LinearInterpolate(A=@n6, Alpha=@n8), B=LinearInterpolate(A=@n9, Alpha=@n10))) {name=UV.V Scalar, nodeColor=<rgba 1 0.3984 0 1>}
  n11 = NamedRerouteUsage()
  n12 = NamedRerouteUsage()
  n13 = Multiply(
   A=DynamicParameter() {paramNames [Width Scalar, Sin Amplitude, Opacity, Growth]}.Width Scalar
   B=LinearInterpolate(
    A=ScalarParameter() {defaultValue=0.005, group=Lightning, parameterName=Lightning Width Min, sortPriority=2}
    B=ScalarParameter() {defaultValue=0.025, group=Lightning, parameterName=Lightning Width Max, sortPriority=3}
    Alpha=Saturate(None=@n12)
   ) {constB=0.1}
  )
  n14 = NamedRerouteUsage()
  n15 = TextureCoordinate()
  n16 = AppendVector(A=ComponentMask(None=@n15) {r=true}, B=Add(A=ComponentMask(None=@n15) {g=true}, B=ParticleRandom()))
  n17 = Divide(B=ScalarParameter() {defaultValue=0.08, group=Lightning Noise, parameterName=Noise U Frequency, sortPriority=1})
  n18 = DynamicParameter() {parameterIndex=1, defaultValue=<rgba 0 1 1 1>, paramNames [Dissolve, Noise Frequency, Sin Frequency, Noise Speed]}
  n19 = Divide(B=Multiply(A=ScalarParameter() {defaultValue=0.55, group=Lightning Noise, parameterName=Noise V Frequency, sortPriority=2}, B=@n18.Noise Frequency))
  n20 = ScalarParameter() {defaultValue=0.25, group=Lightning Noise, parameterName=Noise Speed, sortPriority=4}
  n21 = Panner(
   Coordinate=MaterialFunctionCall(UVs=@n16, Texture Scale=AppendVector(A=@n17, B=@n19)) {materialFunction=&ScaleUVsByCenter}.UVs
   Speed=AppendVector(A=Constant(), B=Multiply(A=@n20, B=@n18.Noise Speed))
  )
  n22 = TextureSampleParameter2D(UVs=@n21) {
   group=Lightning Noise
   parameterName=Noise Texture
   samplerType=SAMPLERTYPE_LinearColor
   sortPriority=0
   texture=&/T_Pickaxe_PlatypusBranch_Noise
  }
  n23 = TextureSampleParameter2D(UVs=Add(A=@n21) {constB=0.25}) {
   group=Lightning Noise
   parameterName=Noise Texture
   samplerType=SAMPLERTYPE_LinearColor
   sortPriority=0
   texture=&T_Pickaxe_PlatypusBranch_Noise
  }
  n24 = AppendVector(A=NamedRerouteDeclaration(None=@n22.R) {name=NoiseBase, nodeColor=<rgba 0.4531 0 1 1>}, B=@n23.R)
  n25 = Multiply(A=Subtract(A=Saturate(None=@n24)) {constB=0.5})
  n26 = ScalarParameter() {defaultValue=0.25, group=Lightning Noise, parameterName=Noise Amplitude, sortPriority=3}
  n27 = Multiply(A=@n21) {constB=2}
  n28 = TextureSampleParameter2D(UVs=@n27) {
   group=Lightning Noise
   parameterName=Noise Texture
   samplerType=SAMPLERTYPE_LinearColor
   sortPriority=0
   texture=&T_Pickaxe_PlatypusBranch_Noise
  }
  n29 = TextureSampleParameter2D(UVs=Add(A=@n27) {constB=0.25}) {
   group=Lightning Noise
   parameterName=Noise Texture
   samplerType=SAMPLERTYPE_LinearColor
   sortPriority=0
   texture=&T_Pickaxe_PlatypusBranch_Noise
  }
  n30 = AppendVector(A=@n28.R, B=@n29.R)
  n31 = Multiply(A=Subtract(A=Saturate(None=@n30)) {constB=0.5})
  n32 = Multiply(A=ScalarParameter() {defaultValue=0.25, group=Lightning Noise, parameterName=Noise Amplitude, sortPriority=3}) {constB=0.5}
  n33 = Add(A=@n16, B=Add(A=Multiply(A=@n25, B=@n26), B=Multiply(A=@n31, B=@n32)))
  n34 = NamedRerouteUsage()
  n35 = Multiply(
   A=MaterialFunctionCall!(
    Width=Multiply(A=@n13, B=@n14)
    Sharpness=ScalarParameter() {defaultValue=0.5, group=Lightning, parameterName=Lightning Sharpness, sortPriority=4}
    Offset
    Direction Switch=StaticBool() {value=true}
    Compare
    Input Coordinates=@n33
   ) {materialFunction=/Engine/Functions/Engine_MaterialFunctions03/Procedurals/GeneratedBand}.Result
   B=LinearInterpolate(
    A=ScalarParameter() {group=Lightning, parameterName=Lightning Opacity Min, sortPriority=5}
    B=ScalarParameter() {defaultValue=1, group=Lightning, parameterName=Lightning Opacity Max, sortPriority=6}
    Alpha=Saturate(None=@n34)
   ) {constB=0.1}
  )
  n36 = Multiply(
   A=Multiply(
    A=DynamicParameter() {paramNames [Width Scalar, Sin Amplitude, Opacity, Growth]}.Width Scalar
    B=ScalarParameter() {defaultValue=0.2, group=Lightning Blur, parameterName=Blur Width, sortPriority=0}
   )
   B=NamedRerouteUsage()
  )
  n37 = ScalarParameter() {group=Lightning Blur, parameterName=Blur Sharpness, sortPriority=1}
  n38 = StaticBool() {value=true}
  n39 = Max(
   A=@n35
   B=Multiply(
    A=MaterialFunctionCall(Width=@n36, Sharpness=@n37, Direction Switch=@n38, Input Coordinates=@n33) {materialFunction=&GeneratedBand}.Result
    B=ScalarParameter() {defaultValue=0.1, group=Lightning Blur, parameterName=Blur Opacity, sortPriority=2}
   )
  )
  n40 = TextureCoordinate()
  n41 = DynamicParameter() {paramNames [Width Scalar, Sin Amplitude, Opacity, Growth]}
  n42 = Multiply(
   A=Multiply(A=Multiply(A=@n11, B=@n39), B=DynamicParameter() {paramNames [Width Scalar, Sin Amplitude, Opacity, Growth]}.Opacity)
   B=Add(A=ComponentMask(None=@n40) {g=true}, B=LinearInterpolate(Alpha=@n41.Growth) {constA=-1})
  )
  n43 = DynamicParameter() {parameterIndex=1, defaultValue=<rgba 0 1 1 1>, paramNames [Dissolve, Noise Frequency, Sin Frequency, Noise Speed]}
  n44 = Saturate(None=@n42)
  n45 = Multiply(
   A=@n42
   B=MaterialFunctionCall(
    Transition Phase=@n43.Dissolve
    Height Texture=OneMinus(None=@n44)
    Contrast=ScalarParameter() {group=Opacity, parameterName=Dissolve Contrast}
   ) {materialFunction=&HeightLerp}.Results
  )
  n46 = DepthFade(Opacity=@n45, FadeDistance=ScalarParameter() {defaultValue=10, group=Opacity, parameterName=FadeDistance, sortPriority=0})
  Reroute(None=QualitySwitch(Default=@n46, Low=@n45, High=@n46, Medium=@n45, Epic=@n46)) {desc=OPACITY}
  n47 = Multiply(A=ScalarParameter() {defaultValue=20, desc=Global spline width scalar., group=WPO Spline, parameterName=Spline Width, sortPriority=2})
  n48 = Time()
  n49 = ScalarParameter() {defaultValue=1, group=WPO Oscillation, parameterName=Sin Speed, sortPriority=0}
  n50 = Add(A=NamedRerouteUsage(), B=Multiply(A=Add(A=@n48) {constB=0}, B=@n49) {constB=2})
  n51 = Add(A=NamedRerouteUsage(), B=Multiply(A=Add(A=@n48) {constB=0.1}, B=@n49) {constB=2})
  n52 = Add(A=NamedRerouteUsage(), B=Multiply(A=Add() {constB=0.2}, B=@n49) {constB=2})
  n53 = Transform(None=MaterialFunctionCall(R=Sine(None=@n50), G=Sine(None=@n51), B=Sine(None=@n52)) {materialFunction=&AppendMany}.RGB) {transformSourceType=TRANSFORMSOURCE_Local}
  n54 = Multiply(
   A=DynamicParameter() {paramNames [Width Scalar, Sin Amplitude, Opacity, Growth]}.Sin Amplitude
   B=ScalarParameter() {defaultValue=1, group=WPO Oscillation, parameterName=Sin Amplitude, sortPriority=2}
  )
  Reroute(None=Add(
    A=MaterialFunctionCall(WidthBase=@n47, WidthTip=@n47) {materialFunction=&SplineThicken}.WorldPositionOffset
    B=Multiply(A=@n53, B=@n54)
   )) {desc=WPO}
  n55 = VectorParameter() {
   desc=Set this value to white when using particle color.
   group=Lightning
   parameterName=Lightning Color
   sortPriority=0
   defaultValue=<rgba 1 1 1 1>
  }
  n56 = MaterialFunctionCall(Emissive=Multiply(A=@n55.RGB, B=@n5.RGB)) {materialFunction=&MF_FX_EmissiveAdjustments}
  n57 = Multiply(
   A=Desaturation(None=@n56.Result, Fraction=ScalarParameter() {group=Lightning, parameterName=Lightning Color Core Desat, sliderMax=1})
   B=ScalarParameter() {defaultValue=5, group=Lightning, parameterName=Lightning Color Core Scalar, sortPriority=1}
  )
  Reroute(None=Add(A=@n56.Result, B=Multiply(A=@n57, B=@n35))) {desc=EMISSIVE COLOR}
 ]
}
