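# Idle Niagara effect for the PlatypusBranch pickaxe, layering blade glow, illumination and intermittent lightning, tinted by variant and time-of-day colors.
# BladeGlow holds a burst-spawned mesh, while Illumination pins a camera-offset sprite and variant-colored light to the skeletal mesh.
# Lightning spawns short-lived meshes in randomized loops, animated through curve-driven dynamic material parameters.
# The front-end preview flag adjusts illumination brightness and lightning placement and scale.
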NiagaraSystem /BRCosmetics/Weapons/FORT_Melee/Pickaxe_PlatypusBranch/FX/NS_Pickaxe_PlatypusBranch_Idle
dependencies {
 usedDataInterfaces [
  &NiagaraDataInterfaceArrayFloat
  &NiagaraDataInterfaceCurve
  &NiagaraDataInterfaceSkeletalMesh
  &NiagaraDataInterfaceArrayColor
  &NiagaraDataInterfaceColorCurve
 ]
 usedDynamicInputs [&SelectLinearColorFromArray, &ColorFromCurve]
 usedModules [
  &self:SystemSpawnScript.SystemScriptSource.SystemScriptGraph.NiagaraNodeAssignment_2.SetVariables_#1
  &self:SystemSpawnScript.SystemScriptSource.SystemScriptGraph.NiagaraNodeAssignment_4.SetVariables_#2
  &self:SystemSpawnScript.SystemScriptSource.SystemScriptGraph.NiagaraNodeAssignment_5.SetVariables_#3
  &SystemState
  &Timeline
  &self:SystemSpawnScript.SystemScriptSource.SystemScriptGraph.NiagaraNodeAssignment_1.SetVariables_#4
  &self:SystemSpawnScript.SystemScriptSource.SystemScriptGraph.NiagaraNodeAssignment_3.SetVariables_#5
 ]
}
emitters {
 BladeGlow {
  emitterUpdateScript [
   &EmitterState()
   &SpawnBurst_Instantaneous(Use Spawn Probability=false, UseLoopCountLimit=true)
  ]
  particleSpawnScript [
   &InitializeParticle(
    Color=&MultiplyLinearColor(
     Linear Color=$System.TOD Color
     Scale Factor=&MakeVector4_FromLinearColor(LinearColor=&Multiply_LinearColorByFloat(Linear Color=$System.VariantColor, Float=3))
    )
    Lifetime=5
    UsePositionOffset=false
    Write Material Random=true
   )
   &InitialMeshOrientation(Mesh Axis To Orient=<0 -1 0>, Mesh Orientation Mode=::NewEnumerator2, Use Rotation Vector=false)
  ]
  particleUpdateScript [&ParticleState(Kill Particles When Lifetime Has Elapsed=false)]
  rendererClasses [&NiagaraMeshRendererProperties]
  renderers [{rendererClass=&NiagaraMeshRendererProperties, rendererIndex=0}]
 }
 Illumination {
  emitterUpdateScript [
   &EmitterState(
    Inactive Response=::NewEnumerator1
    Life Cycle Mode=::NewEnumerator1
    Loop Duration=1
    Recalculate Duration Each Loop=true
    UseLoopDelay=false
   )
   &SpawnBurst_Instantaneous(Use Spawn Probability=false, UseLoopCountLimit=true)
  ]
  particleSpawnScript [
   &InitializeParticle(
    Color=&MultiplyLinearColor(
     Linear Color=$System.TOD Color
     Scale Factor=&MakeVector4_FromLinearColor(LinearColor=&Multiply_LinearColorByFloat(
       Linear Color=$System.VariantColor
       Float=&MakeCustomFloatFromBool(Bool=$User.bIsFrontEndPreview, True Float=0.15, False Float=0.3)
      ))
    )
    Sprite Size=<80 165>
    Sprite Size Mode=::NewEnumerator4
    UsePositionOffset=true
    Write Material Random=true
   )
   &CameraOffset(Camera Offset Amount=25)
  ]
  particleUpdateScript [
   &ParticleState(Kill Particles When Lifetime Has Elapsed=false, Let Infinitely Lived Particles Die When Emitter Deactivates=true)
   &SkeletalMeshLocation(
    Bone / Socket Index=1
    MaskSpawnGroup=false
    Mesh Sampling Type=::NewEnumerator5
    Skeletal Mesh=$System.Skeletal Mesh
    Socket Sampling Mode=::NewEnumerator5
   )
   &SpriteFacingAndAlignment(Sprite Alignment=<0 0 1>, WriteAlignment=true, WriteFacing=false)
   &FN_Light_Attributes(
    Color=$System.VariantColor
    Create Color Attribute=true
    Create Exponent Attribute=true
    Create Position Attribute=true
    Create Radius Attribute=true
    Exponent=3
    Radius=40
    Scale Color=10
   )
  ]
  rendererClasses [&NiagaraSpriteRendererProperties, &NiagaraLightRendererProperties]
  renderers [
   {rendererClass=&NiagaraSpriteRendererProperties}
   {rendererClass=&NiagaraLightRendererProperties, rendererIndex=1}
  ]
 }
 Lightning {
  emitterUpdateScript [
   &EmitterState(
    Inactive Response=::NewEnumerator1
    Life Cycle Mode=::NewEnumerator1
    Loop Delay=&RandomRangeFloat(Minimum=0, Maximum=0.25)
    Loop Duration=&RandomRangeFloat(Minimum=0, Maximum=1)
    Recalculate Duration Each Loop=true
    UseLoopDelay=true
   )
   &SpawnRate(
    SpawnRate=&Multiply_Float(A=&RandomRangeFloat(Minimum=2, Maximum=6), B=2)
    Use Spawn Probability=false
   )
   /SetVariables(Emitter.ColorCoreScalar:NiagaraFloat=&SelectFloatFromArray(
     Direct Array Index=&MakeIntFromFloat(Float=$User.VariantID)
     Float Selection Array=DataInterface() {propertyValues {FloatData [1,15]}}
    )) {isSetParameters=true}
  ]
  particleSpawnScript [
   /SetVariables(Particles.GlobalRandom:NiagaraFloat=&RandomRangeFloat(Minimum=0.4, Maximum=1)) {isSetParameters=true}
   &InitializeParticle(
    Color=&MultiplyLinearColor(
     Linear Color=$System.TOD Color
     Scale Factor=&MakeVector4_FromLinearColor(LinearColor=&Multiply_LinearColorByFloat(Linear Color=$System.VariantColor, Float=1))
    )
    Lifetime=&Multiply_Float(A=&RandomRangeFloat(Minimum=0.2, Maximum=1), B=1)
    Mesh Scale=&MakeCustomVectorFromBool(Bool=$User.bIsFrontEndPreview, True Vector=<0.9 0.9 0.94>, False Vector=<0.9 0.9 0.9>)
    Mesh Scale Mode=::NewEnumerator4
    Position Offset=/Niagara/DynamicInputs/TypeConversions/MakeCustomPositionFromBool(Bool:NiagaraBool=$User.bIsFrontEndPreview, True Position:NiagaraPosition=<0 0.1 1>, False Position:NiagaraPosition=<0 -1 5>)
    UsePositionOffset=true
    Write Material Random=true
   )
   &InitialMeshOrientation(Mesh Orientation Mode=::NewEnumerator7, Rotation=<0 0 0.25>, Use Rotation Vector=true)
   &CameraOffset(Camera Offset Amount=10)
  ]
  particleUpdateScript [
   &ParticleState()
   &DynamicMaterialParameters(
    Index 0 Param 1=&FloatFromCurve(
     FloatCurve=DataInterface() {propertyValues {Curve {
        defaultValue=0
        keys [
         {arriveTangent=7.2591, interpMode=RCIM_Cubic, leaveTangent=7.2591, tangentMode=RCTM_User, time=0, value=0}
         {arriveTangent=-2.5249, interpMode=RCIM_Cubic, leaveTangent=-2.5249, tangentMode=RCTM_User, time=0.25, value=0.6}
         {arriveTangent=-0.8477, interpMode=RCIM_Cubic, leaveTangent=-0.8477, tangentMode=RCTM_User, time=1, value=0}
        ]
       }}}
     Scale Curve=&RandomRangeFloat(Minimum=1, Maximum=3)
    )
    Index 0 Param 2=&FloatFromCurve(
     FloatCurve=DataInterface() {propertyValues {Curve {
        defaultValue=0
        keys [<at 0 0.2 RCIM_Cubic>, {arriveTangent=2.4937, interpMode=RCIM_Cubic, leaveTangent=2.4937, tangentMode=RCTM_User, time=1, value=1}]
       }}}
     Scale Curve=&Multiply_Float(A=5, B=$Particles.GlobalRandom)
    )
    Index 0 Param 3=&RandomRangeFloat(Minimum=0.5, Maximum=1)
    Index 0 Param 4=&FloatFromCurve(FloatCurve=DataInterface() {propertyValues {Curve {
        defaultValue=0
        keys [
         {arriveTangent=4.1785, interpMode=RCIM_Cubic, leaveTangent=4.1785, tangentMode=RCTM_User, time=0, value=0}
         {arriveTangent=0.0224, interpMode=RCIM_Cubic, leaveTangent=0.0224, tangentMode=RCTM_User, time=0.5, value=1}
        ]
       }}})
    Index 1 Param 1=&FloatFromCurve(FloatCurve=DataInterface() {propertyValues {Curve {
        defaultValue=1
        keys [
         {interpMode=RCIM_Cubic, tangentMode=RCTM_User, time=0, value=0}
         {arriveTangent=2.9607, interpMode=RCIM_Cubic, leaveTangent=2.9607, tangentMode=RCTM_User, time=0.7, value=1}
        ]
       }}})
    Index 1 Param 2=&Multiply_Float(A=&MaskFloatBySpawnGroup(Masked Float=0.8, Passthrough Float=0.4), B=$Particles.GlobalRandom)
    Index 1 Param 3=&Multiply_Float(A=&MaskFloatBySpawnGroup(Masked Float=0.4, Passthrough Float=0.3), B=$Particles.GlobalRandom)
    Index 1 Param 4=&RandomRangeFloat(Minimum=0.05, Maximum=0.15)
    Module.Index0Float0Write=true
    Module.Index0Float1Write=true
    Module.Index0Float2Write=true
    Module.Index0Float3Write=true
    Module.Index1Float0Write=true
    Module.Index1Float1Write=true
    Module.Index1Float2Write=true
    Module.Index1Float3Write=true
    Write Parameter Index 1=::NewEnumerator1
   )
  ]
  rendererClasses [&NiagaraMeshRendererProperties]
  renderers [{rendererClass=&NiagaraMeshRendererProperties, rendererIndex=0}]
 }
}
summary {
 systemName=NS_Pickaxe_PlatypusBranch_Idle
 emitters [
  {emitterName=BladeGlow, rendererClasses [&NiagaraMeshRendererProperties]}
  {emitterName=Illumination, rendererClasses [&NiagaraSpriteRendererProperties, &NiagaraLightRendererProperties]}
  {emitterName=Lightning, rendererClasses [&NiagaraMeshRendererProperties]}
 ]
 userVariables [
  User.bIsFrontEndPreview {type=NiagaraBool}
  User.bisSwinging {type=NiagaraBool}
  User.bWeakpointHit {type=NiagaraBool}
  User.LoopsDelay {type=NiagaraFloat, defaultValue=0, description=""}
  User.ReactivitySwitch {type=NiagaraFloat, defaultValue=0, description=""}
  User.VariantID {type=NiagaraFloat, defaultValue=0, description=""}
 ]
}
